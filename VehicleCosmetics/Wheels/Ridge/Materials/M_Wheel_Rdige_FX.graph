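# Animated energy-effect material for the Ridge wheel cosmetic, tinted by a CustomColor parameter.
# Bump-offset radial coordinates give parallax depth to panned textures that build a stringy energy pattern, with optional sparkles behind a static switch.
# Scrolling wisps and a flow-distorted outer band add to it, and radial falloff and packed-texture masks shape the glow.
# The result, plus a vertex-color term, drives a blend from BaseColor to CustomColor.
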Material /VehicleCosmetics/Wheels/Ridge/Materials/M_Wheel_Rdige_FX
graph {
 properties {bUsedWithMeshDeformer=true, bUsedWithSkeletalMesh=true, pixelDepthOffsetMode=PDOM_Legacy, usedShadingModels=Default Lit}
 nodes [
  Constant()
  n1 = VectorParameter() {parameterName=CustomColor, defaultValue=<rgba 1 0.5333 0 0.1569>}
  n2 = TextureCoordinate()
  n3 = ComponentMask(None=Add(A=@n2, B=Constant() {r=-0.5})) {g=true, r=true}
  n4 = DotProduct(A=@n3, B=@n3)
  n5 = Multiply(A=Clamp(None=@n4), B=VertexColor())
  n6 = OneMinus(None=SquareRoot(None=@n4))
  n7 = ScalarParameter() {defaultValue=5, parameterName=EnergyAmount}
  n8 = BumpOffset(
   Coordinate=ComponentMask(None=TextureSample() {samplerType=SAMPLERTYPE_LinearColor, texture=/VehicleCosmetics/Wheels/Ridge/Textures/T_Radial_Coords_Nearest}.RGB) {g=true, r=true}
   Height=OneMinus(None=TextureSample() {texture=/VehicleCosmetics/Wheels/Ridge/Textures/T_Radial_Generic_01_Pack}.B)
  ) {heightRatio=-2, referencePlane=1}
  n9 = Multiply(A=@n8, B=Constant2Vector() {g=1, r=3})
  n10 = Time()
  n11 = Multiply(A=@n10, B=ScalarParameter() {defaultValue=3, parameterName=Speed})
  n12 = TextureSample(UVs=Add(
    A=Panner(Coordinate=@n9, Time=@n11) {speedY=0.05}
    B=TextureSample(UVs=@n8) {texture=/VehicleCosmetics/Wheels/Ridge/Textures/T_Water_01_nomips}.G
   )) {samplerType=SAMPLERTYPE_LinearColor, texture=/VehicleCosmetics/Wheels/Ridge/Textures/T_OOB_Planet00_N_nomips}
  n13 = Add(
   A=@n12.A
   B=TextureSample(UVs=Panner(Coordinate=@n8, Time=@n11) {speedY=0.02}) {samplerType=SAMPLERTYPE_LinearColor, texture=&T_OOB_Planet00_N_nomips}.A
  )
  n14 = Panner(Coordinate=Multiply(A=@n8, B=Constant2Vector() {g=0.2, r=0.8})) {speedY=-0.02}
  n15 = Multiply(A=@n13, B=Multiply(A=TextureSample(UVs=@n14) {texture=/VehicleCosmetics/Wheels/Ridge/Textures/T_Noise_Spots_RGB_nomips}.B) {constB=16})
  n16 = Power(Base=@n13, Exp=ScalarParameter() {defaultValue=1, parameterName=StringyAmount})
  n17 = Multiply(A=@n8, B=Constant2Vector() {g=1, r=2})
  n18 = TextureSample() {texture=&T_Radial_Generic_01_Pack}
  n19 = Panner(Coordinate=Add(A=@n17, B=@n18.B), Time=Add(A=@n10, B=@n18.B)) {speedX=-1.5, speedY=-0.1}
  n20 = Multiply(
   A=Power(
    Base=TextureSample(UVs=@n19) {texture=/VehicleCosmetics/Wheels/Ridge/Textures/T_Ridge_SupernovaTrail_RGB}.G
    Exp=ScalarParameter() {defaultValue=3, parameterName=EnergyWisps_Amount}
   )
   B=ScalarParameter() {defaultValue=48, parameterName=EnergyWisps_Intensity}
  )
  n21 = OneMinus(None=@n18.G)
  n22 = Saturate(None=Add(A=@n21, B=@n21))
  n23 = ComponentMask(None=TextureSample(UVs=Panner(Coordinate=@n17) {speedY=-0.3}) {texture=/VehicleCosmetics/Wheels/Ridge/Textures/T_Flow_03_nomips}.RGB) {g=true, r=true}
  n24 = ComponentMask(None=Add(A=Add(A=@n2, B=@n23), B=ScalarParameter() {defaultValue=-0.7, parameterName=Outer_Centered}) {constB=-0.5}) {g=true, r=true}
  n25 = DotProduct(A=@n24, B=@n24)
  n26 = ScalarParameter() {defaultValue=6, parameterName=Outer_Amount}
  n27 = Saturate(None=Multiply(A=Power(Base=@n25, Exp=@n26), B=ScalarParameter() {defaultValue=2048, parameterName=Outer_Hardness}))
  n28 = ScalarParameter() {defaultValue=3, parameterName=Outer_Intensity}
  n29 = Add(
   A=StaticSwitchParameter(True=Add(A=@n15, B=@n16), False=@n13) {parameterName=Sparkles}
   B=Add(A=Multiply(A=@n20, B=@n22), B=Multiply(A=@n27, B=@n28))
  )
  n30 = Multiply(B=@n1.RGB) {constA=2}
  n31 = Multiply(A=Multiply(A=Power(Base=@n6, Exp=@n7), B=Multiply(A=@n29, B=@n30)), B=@n22)
  LinearInterpolate(
   A=VectorParameter() {parameterName=BaseColor, defaultValue=<rgba 0.058 0.0264 0.0258 0.2314>}.RGB
   B=@n1.RGB
   Alpha=Power(Base=Add(A=@n5, B=@n31))
  )
  Multiply(A=@n12.A)
  TextureSample() {samplerType=SAMPLERTYPE_Normal, texture=/VehicleCosmetics/Wheels/Ridge/Textures/T_Sphere_NoMin_N}
 ]
}
